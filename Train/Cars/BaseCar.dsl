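; Base train car Blueprint that handles a car's color, audio and player dust effects.
; On cosmetic BeginPlay it destroys WheelAudio-tagged components on mobile and Switch, then writes the train manager's color index into the mesh's custom primitive data.
; Moving-platform events disable running dust while players are aboard, with debug prints.
; It can restart all car audio, and hides the sound indicator for players based on its mesh.
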(event EventBeginPlay
  (bind _returnvalue (CallFunction|GetTrainManager))
  (Networking|CanExecuteCosmeticEvents
    (:True
      (Utilities|FlowControl|Sequence
        (:then_0
          (if (or (Fortnite|IsMobilePlatform) (TouchPlatform|IsNintendoSwitch))
            (for _array_element (Actor|GetComponentsbyTag "/Script/Engine.AudioComponent" "WheelAudio")
              (Utilities|IsValid _array_element
                (:Is%20Valid
                  (Audio|Components|Audio|Stop _array_element)
                  (Transformation|DetachFromComponent _array_element "KeepRelative" "KeepRelative" "KeepRelative" true)
                  (Components|DestroyComponent _array_element))))))
        (:then_1
          ;[1 1 1] Update color of car
          (Utilities|IsValid _returnvalue
            (:Is%20Valid
              (bind _asbp_default_train_manager (Utilities|Casting|CastToBP_DefaultTrainManager _returnvalue))
              (CallFunction|SetMeshColorIndex (Variables|/Train/BP_DefaultTrainManager.BP_DefaultTrainManager_C|GetColorIndex _asbp_default_train_manager)))))))))

(event Custom|RestartTrainAudio
  (for _array_element (Actor|GetComponentsByClass "/Script/Engine.AudioComponent")
    (Audio|Components|Audio|Play _array_element 0.0)))

(event OnPawnBasedonMovingPlatform(FortTrainMovingPlatform)
  (params (MovingPlatform :type Fort%20Base%20Moving%20Platform%20Component%20Object%20Reference) (Pawn :type Fort%20Pawn%20Object%20Reference))
  ;[1 1 1] Handle the player running dust FX to turn On/Off
  (Networking|CanExecuteCosmeticEvents
    (:True
      (Development|PrintString "OnTrain" true false "(R=1.000000,G=0.015783,B=0.000000,A=1.000000)" 5.0)
      (bind _asfort_player_pawn_athena (Utilities|Casting|CastToFortPlayerPawnAthena Pawn))
      (Variables|/Script/FortniteGame.FortPlayerPawn|SetbFXPlayDustOnMovement false _asfort_player_pawn_athena))))

(event OnPawnDebasedMovingPlatform(FortTrainMovingPlatform)
  (params (MovingPlatform :type Fort%20Base%20Moving%20Platform%20Component%20Object%20Reference) (Pawn :type Fort%20Pawn%20Object%20Reference) (NewBase :type Actor%20Object%20Reference))
  ;[1 1 1] Handle the player running dust FX to turn On/Off
  (Networking|CanExecuteCosmeticEvents
    (:True
      (Development|PrintString "OffTrain" true false "(R=0.020150,G=1.000000,B=0.000000,A=1.000000)" 5.0)
      (bind _asfort_player_pawn_athena (Utilities|Casting|CastToFortPlayerPawnAthena Pawn))
      (Variables|/Script/FortniteGame.FortPlayerPawn|SetbFXPlayDustOnMovement true _asfort_player_pawn_athena))))

(fn SetMeshColorIndex
  (params (ColorIndex :type Integer))
  (if (>= (Variables|Default|GetColorMaterialIndex) 0)
    (Rendering|Material|SetCustomPrimitiveDataFloat (Variables|Default|GetStaticMesh) (Variables|Default|GetColorMaterialIndex) (Math|Conversions|ToFloat(Integer) ColorIndex))))

(fn ShouldShowSoundIndicator
  (params (PlayerController :type Fort%20Player%20Controller%20Object%20Reference))
  (bind _returnvalue (Game|Player|GetPlayerPawn PlayerController))
  (Utilities|IsValid _returnvalue
    (:Is%20Valid
      (return (not (== (Pawn|Components|CharacterMovement|GetMovementBaseObject (Variables|/Script/Engine.Character|GetCharacterMovement _returnvalue)) (Variables|Default|GetStaticMesh)))))
    (:Is%20Not%20Valid
      (return false))))

(fn ConstructionScript)
