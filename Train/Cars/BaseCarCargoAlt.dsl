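; Alternate cargo train car Blueprint that manages looping-audio cleanup for its two onboard loot containers.
; On BeginPlay it casts both child-actor chests to the tiered chest class and binds their OnDied events.
; Each death sets a RepNotify flag, whose OnRep destroys the matching audio component.
; The audio components live on the car rather than the child actors for better performance.
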(event EventBeginPlay
  (|CallParentFunction)
  ;[0 0.1312 1] Audio Components for the Loot Containers live in the train car for better perf vs them being on the child actor.
  ; Destroy associated audio component on child actor death
  (bind _astiered_chest_athena_parent (Utilities|Casting|CastToTiered_Chest_Athena_Parent (Variables|/Script/Engine.ChildActorComponent|GetChildActor (Variables|Default|GetTrain_OnBoard_LootContainer1))))
  (EventDispatchers|BindEventtoOnDied _astiered_chest_athena_parent (EventDispatchers|CreateEvent :!event Custom|DestroyLootContainerAudioComponent))
  (bind _astiered_chest_athena_parent_2 (Utilities|Casting|CastToTiered_Chest_Athena_Parent (Variables|/Script/Engine.ChildActorComponent|GetChildActor (Variables|Default|GetTrain_OnBoard_LootContainer2))))
  (EventDispatchers|BindEventtoOnDied _astiered_chest_athena_parent_2 (EventDispatchers|CreateEvent :!event Custom|DestroyLootContainerAudioComponentTwo)))

(event Collision|EventActorBeginOverlap
  (params (OtherActor :type Actor%20Object%20Reference)))

(event EventTick
  (params (DeltaSeconds :type Float%20%28single-precision%29)))

(event Custom|DestroyLootContainerAudioComponent
  (params (DamagedActor :type Actor%20Object%20Reference) (Damage :type Float%20%28single-precision%29) (InstigatedBy :type Controller%20Object%20Reference) (DamageCauser :type Actor%20Object%20Reference) (HitLocation :type Vector) (FHitComponent :type Primitive%20Component%20Object%20Reference) (BoneName :type Name) (Momentum :type Vector))
  (Variables|Components|SetRepNotifyDestroyContainerOne true))

(event Custom|DestroyLootContainerAudioComponentTwo
  (params (DamagedActor :type Actor%20Object%20Reference) (Damage :type Float%20%28single-precision%29) (InstigatedBy :type Controller%20Object%20Reference) (DamageCauser :type Actor%20Object%20Reference) (HitLocation :type Vector) (FHitComponent :type Primitive%20Component%20Object%20Reference) (BoneName :type Name) (Momentum :type Vector))
  (Variables|Default|SetRepNotifyDestroyContainerTwo true))

(fn OnRep_RepNotifyDestroyContainerOne
  (Components|DestroyComponent (Variables|Default|GetLootContainerAudio1)))

(fn OnRep_RepNotifyDestroyContainerTwo
  (Components|DestroyComponent (Variables|Default|GetLootContainerAudio2)))

(fn ConstructionScript
  (|CallParentFunction))
